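# Translucent, unlit, two-sided Niagara mesh-particle material that draws a smoky motion-blur streak on the rope in the Hoist emote.
# Dynamic parameters RopePosition and BlurSize define a wrapping band along V, shaped by a reflected gradient and a smoke-trail texture.
# Opacity also fades at the U edges, at grazing view angles and with particle alpha, while particle color feeds emissive.
# An offset collapses vertices outside the band to the particle position.
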Material /BRCosmetics/Animation/Game/MainPlayer/Emotes/Hoist/FX/Materials/M_Emote_Hoist_RopeBlur
graph {
 properties {
  bUsedWithNiagaraMeshParticles=true
  blendMode=BLEND_Translucent
  pixelDepthOffsetMode=PDOM_Legacy
  shadingModel=MSM_Unlit
  twoSided=true
  usedShadingModels=Unlit
 }
 nodes [
  n1 = TextureCoordinate()
  n2 = DynamicParameter() {defaultValue=<rgba 0.9 0.2 1 1>, paramNames [RopePosition,BlurSize,Param3,Param4]}
  n3 = Frac(None=Subtract(A=ComponentMask(None=@n1) {g=true}, B=@n2.RopePosition))
  n4 = OneMinus(None=Saturate(None=Divide(A=@n3, B=@n2.BlurSize)))
  Frac(None=@n4)
  n5 = ParticleColor()
  MaterialFunctionCall(Emissive=@n5.RGB) {materialFunction=&MF_FX_EmissiveAdjustments}
  n6 = ScalarParameter() {defaultValue=0.9, parameterName=Gradient Center Point}
  n7 = ComponentMask(None=@n1) {r=true}
  n8 = Multiply(
   A=Power(
    Base=MaterialFunctionCall!(Gradient=@n4, Center Point=@n6) {materialFunction=/Game/Effects/Fort_Effects/MaterialFunctions/MF_GradientReflected}.Result
    Exp=ScalarParameter() {defaultValue=3, parameterName=Gradient Exponent}
   )
   B=TextureSample(UVs=AppendVector(A=@n7, B=@n4)) {texture=/BRCosmetics/Weapons/FORT_Melee/Pickaxe_Llama_Hero_Winter_Male/FX/T_Puffy_Smoke_Trail_01}.R
  )
  n9 = ConstantBiasScale(None=@n7) {bias=-0.5, scale=2}
  n10 = OneMinus(None=DotProduct(A=@n9, B=@n9))
  n11 = DotProduct(A=VertexNormalWS(), B=CameraVectorWS())
  Multiply(A=@n5.A, B=Multiply(A=Multiply(A=@n8, B=@n10), B=Abs(None=@n11)))
  Multiply(A=OneMinus(None=Ceil(None=@n4)), B=Subtract(A=ParticlePositionWS(), B=WorldPosition().XYZ))
 ]
}
