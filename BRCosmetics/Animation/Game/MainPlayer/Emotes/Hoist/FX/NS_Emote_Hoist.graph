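# Niagara system for the Hoist emote that puffs sprite dust at the character's feet.
# Each update samples two skeletal mesh foot points, tests their height against a threshold and compares with the previous result to reset emitter age, retriggering per-foot burst spawns.
# Spawn groups place rings near ground level under each foot.
# Particles burst outward, slow with drag, drift upward and fade by system opacity, with curve-driven dynamic material parameters.
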NiagaraSystem /BRCosmetics/Animation/Game/MainPlayer/Emotes/Hoist/FX/NS_Emote_Hoist
dependencies {
 usedDataInterfaces [&NiagaraDataInterfaceCurve, &NiagaraDataInterfaceSkeletalMesh]
 usedDynamicInputs [&ClampFloat, &Subtract_Float, &Multiply_Float]
 usedModules [
  &self:SystemSpawnScript.SystemScriptSource.SystemScriptGraph.NiagaraNodeAssignment_3.SetVariables_#1
  &SystemState
  &self:SystemSpawnScript.SystemScriptSource.SystemScriptGraph.NiagaraNodeAssignment_0.SetVariables_#2
 ]
}
emitters {DustPuff {
  emitterSpawnScript [/SetVariables() {isSetParameters=true}]
  emitterUpdateScript [
   &EmitterState(
    Inactive Response=::NewEnumerator1
    Life Cycle Mode=::NewEnumerator1
    Loop Behavior=::NewEnumerator1
    Loop Duration Mode=::NewEnumerator1
    UseLoopDelay=false
   )
   &SampleSkeletalMesh(
    AdditionalYawPitchRoll=false
    Bone Sampling Mode:ENiagaraBoneSamplingMode=::NewEnumerator4
    Skeletal Mesh:NiagaraDataInterfaceSkeletalMesh=$System.Skeletal Mesh
   )
   &SampleSkeletalMesh(
    AdditionalYawPitchRoll=false
    Bone / Socket Index=1
    Bone Sampling Mode:ENiagaraBoneSamplingMode=::NewEnumerator4
    Skeletal Mesh:NiagaraDataInterfaceSkeletalMesh=$System.Skeletal Mesh
   ) {module=SampleSkeletalMesh001}
   /SetVariables(
    Emitter.SpawnConditionLeftFoot:NiagaraBool=&SetBoolByFloatComparison(
     A=&MakeFloatFromVector(Vector=$Emitter.SampleSkeletalMesh.SampledPosition, Channel=::NewEnumerator2)
     B=15
     Comparison Type=::NewEnumerator7
    )
    Emitter.SpawnConditionRightFoot:NiagaraBool=&SetBoolByFloatComparison(
     A=&MakeFloatFromVector(Vector=$Emitter.SampleSkeletalMesh001.SampledPosition, Channel=::NewEnumerator2)
     B=15
     Comparison Type=::NewEnumerator7
    )
   ) {isSetParameters=true}
   /SetVariables(
    Emitter.CompareSpawnConditionsLeftFoot:NiagaraBool=&SetBoolByFloatComparison(
     A=&MakeCustomFloatFromBool(Bool=$Emitter.SpawnConditionLeftFoot, True Float=1, False Float=0)
     B=&MakeCustomFloatFromBool(Bool=$Emitter.PreviousSpawnConditionLeftFoot, True Float=1, False Float=0)
    )
    Emitter.CompareSpawnConditionsRightFoot:NiagaraBool=&SetBoolByFloatComparison(
     A=&MakeCustomFloatFromBool(Bool=$Emitter.SpawnConditionRightFoot, True Float=1, False Float=0)
     B=&MakeCustomFloatFromBool(Bool=$Emitter.PreviousSpawnConditionRightFoot, True Float=1, False Float=0)
    )
   ) {isSetParameters=true}
   /SetVariables(Emitter.Age:NiagaraFloat=&MakeCustomFloatFromBool(
     Bool=&Boolean_OrOperation(Bool1=$Emitter.CompareSpawnConditionsLeftFoot, Bool2=$Emitter.CompareSpawnConditionsRightFoot)
     True Float=0
     False Float=$Emitter.Age
    )) {isSetParameters=true}
   &SpawnBurst_Instantaneous(
    Age:NiagaraFloat=$Emitter.Age
    Spawn Count=7
    Spawn Probability=&MakeCustomFloatFromBool(Bool=$Emitter.SpawnConditionLeftFoot, True Float=1, False Float=0)
    Spawn Time=0.05
    Use Spawn Probability=true
    UseLoopCountLimit=false
   )
   &SpawnBurst_Instantaneous(
    Age:NiagaraFloat=$Emitter.Age
    Spawn Count=7
    Spawn Group=1
    Spawn Probability=&MakeCustomFloatFromBool(Bool=$Emitter.SpawnConditionRightFoot, True Float=1, False Float=0)
    Spawn Time=0.05
    Use Spawn Probability=true
    UseLoopCountLimit=false
   ) {module=SpawnBurst_Instantaneous001}
   /SetVariables(
    Emitter.InitialParticlePositionLeftFoot:NiagaraPosition=&MakeVector(
     X=&MakeFloatFromVector(Vector=$Emitter.SampleSkeletalMesh.SampledPosition, Channel=::NewEnumerator0)
     Y=&MakeFloatFromVector(Vector=$Emitter.SampleSkeletalMesh.SampledPosition, Channel=::NewEnumerator1)
     Z=4
    )
    Emitter.InitialParticlePositionRightFoot:NiagaraPosition=&MakeVector(
     X=&MakeFloatFromVector(Vector=$Emitter.SampleSkeletalMesh001.SampledPosition, Channel=::NewEnumerator0)
     Y=&MakeFloatFromVector(Vector=$Emitter.SampleSkeletalMesh001.SampledPosition, Channel=::NewEnumerator1)
     Z=4
    )
    Emitter.PreviousSpawnConditionLeftFoot:NiagaraBool=$Emitter.SpawnConditionLeftFoot
    Emitter.PreviousSpawnConditionRightFoot:NiagaraBool=$Emitter.SpawnConditionRightFoot
   ) {isSetParameters=true}
  ]
  particleSpawnScript [
   &InitializeParticle(
    Color=<rgba 0.2 0.1776 0.168 0.2>
    Lifetime Max=0.5
    Lifetime Min=0.25
    Lifetime Mode=::NewEnumerator1
    Position=&MaskVectorBySpawnGroup(
     Masked Vector=Linked() {linkedVariable {name=Emitter.InitialParticlePositionLeftFoot, type=NiagaraPosition}}
     Passthrough Vector=Linked() {linkedVariable {name=Emitter.InitialParticlePositionRightFoot, type=NiagaraPosition}}
    )
    Position Mode=::NewEnumerator1
    Sprite Rotation Mode=::NewEnumerator2
    Sprite Size Mode=::NewEnumerator1
    Uniform Sprite Size Max=26
    Uniform Sprite Size Min=16
    UsePositionOffset=false
    Write Material Random=true
   )
   &ShapeLocation(
    MaskSpawnGroup=false
    Offset=<0 11 0>
    Offset Mode=::NewEnumerator0
    Override Seed=false
    Ring / Disc Distribution Mode:ENiagaraSphereDistributionMode=::NewEnumerator1
    Ring Radius=6
    Shape Primitive=::NewEnumerator4
    U Position=&ReturnNormalizedExecIndex()
    UseOwnerScale=true
   )
   &AddVelocity(
    Module.Use Falloff Distance=false
    Origin Offset=<0 11 -1>
    Override Seed=false
    Velocity Mode=::NewEnumerator1
    Velocity Origin=&MaskVectorBySpawnGroup(
     Masked Vector=Linked() {linkedVariable {name=Emitter.InitialParticlePositionLeftFoot, type=NiagaraPosition}}
     Passthrough Vector=Linked() {linkedVariable {name=Emitter.InitialParticlePositionRightFoot, type=NiagaraPosition}}
    )
    Velocity Speed=&RandomRangeFloat(Minimum=55, Maximum=99)
   )
   &CameraOffset(Camera Offset Amount=15)
  ]
  particleUpdateScript [
   &ParticleState()
   &Drag(Drag=7, Use Linear Drag=true, Use Rotational Drag=true)
   &GravityForce(Gravity=<0 0 222>)
   &SolveForcesAndVelocity(Clamp Velocity=false, Limit Acceleration=false)
   &ScaleColor(
    Scale Alpha=&FloatFromCurve(
     FloatCurve=DataInterface() {propertyValues {Curve {keys [{interpMode=RCIM_Cubic, leaveTangent=-1.7202, tangentMode=RCTM_Break, time=0, value=1}, {tangentMode=RCTM_User, time=1, value=0}]}}}
     Scale Curve=$System.Opacity
    )
    Scale RGB=&VectorFromFloat(Value=&FloatFromCurve(FloatCurve=DataInterface() {propertyValues {Curve {keys [<at 0 1>,<at 0.9983 0.5>]}}}))
    ScaleA=true
    ScaleRGB=true
   )
   &ScaleSpriteSize(Scale Factor=&Vector2DFromFloat(Value=&FloatFromCurve(FloatCurve=DataInterface() {propertyValues {Curve {keys [
          {arriveTangent=1.4704, interpMode=RCIM_Cubic, leaveTangent=1.4704, tangentMode=RCTM_User, time=0, value=0.5}
          {arriveTangent=-0.0405, interpMode=RCIM_Cubic, leaveTangent=-0.0405, tangentMode=RCTM_User, time=1, value=1}
         ]}}})))
   &DynamicMaterialParameters(
    Index 0 Param 1=&Add_Float(
     A=$Particles.RandomFloat
     B=&FloatFromCurve(
      FloatCurve=DataInterface() {propertyValues {Curve {keys [
          {arriveTangent=0.9088, interpMode=RCIM_Cubic, leaveTangent=0.9088, tangentMode=RCTM_User, time=0, value=1}
          {arriveTangent=0.5473, interpMode=RCIM_Cubic, leaveTangent=0.5473, tangentMode=RCTM_User, time=1, value=2}
         ]}}}
      CurveIndex=$Particles.Age
      Scale Curve=0.222
     )
    )
    Index 0 Param 2=&FloatFromCurve(FloatCurve=DataInterface() {propertyValues {Curve {keys [
         {arriveTangent=0.0453, interpMode=RCIM_Cubic, tangentMode=RCTM_Break, time=0, value=-1}
         {arriveTangent=0.0604, interpMode=RCIM_Cubic, leaveTangent=0.0604, tangentMode=RCTM_User, time=1, value=-1}
        ]}}})
    Index 0 Param 3=&FloatFromCurve(FloatCurve=DataInterface() {propertyValues {Curve {keys [
         {arriveTangent=-9.0447, interpMode=RCIM_Cubic, leaveTangent=-2.0044, tangentMode=RCTM_User, time=0, value=0.5}
         {interpMode=RCIM_Cubic, leaveTangent=4.6779, time=0.2851, value=0}
        ]}}})
    Index 1 Param 1=&FloatFromCurve(
     FloatCurve=DataInterface() {propertyValues {Curve {keys [<at 0 0>,<at 0.25 1>]}}}
     Scale Curve=&RandomRangeFloat(Minimum=1, Maximum=1)
    )
    Index 1 Param 2=&FloatFromCurve(
     FloatCurve=DataInterface() {propertyValues {Curve {keys [
         {interpMode=RCIM_Cubic, leaveTangent=3.7422, tangentMode=RCTM_Break, time=0, value=0.2}
         {interpMode=RCIM_Cubic, leaveTangent=0.9066, time=1, value=2}
        ]}}}
     Scale Curve=&RandomRangeFloat(Minimum=0.5, Maximum=1)
    )
    Index 1 Param 3=&FloatFromCurve(FloatCurve=DataInterface() {propertyValues {Curve {keys [
         {interpMode=RCIM_Cubic, leaveTangent=2.8445, tangentMode=RCTM_Break, time=0, value=0.2}
         {interpMode=RCIM_Cubic, leaveTangent=63.2978, time=0.25, value=1}
        ]}}})
    Index 1 Param 4=&FloatFromCurve(
     FloatCurve=DataInterface() {propertyValues {Curve {keys [<at 0 1>,<at 1 0.8>]}}}
     Scale Curve=&RandomRangeFloat(Minimum=0.5, Maximum=1)
    )
    Index 2 Param 1=&FloatFromCurve(
     FloatCurve=DataInterface() {propertyValues {Curve {keys [{interpMode=RCIM_Cubic, leaveTangent=3.2309, tangentMode=RCTM_Break, time=0.1, value=0}, <at 1 1>]}}}
     Scale Curve=0
    )
    Index 2 Param 2=&FloatFromCurve(
     FloatCurve=DataInterface() {propertyValues {Curve {keys [<at 0 0.2 RCIM_Cubic>,<at 1 0.8 RCIM_Cubic>]}}}
     Scale Curve=&RandomRangeFloat(Minimum=0.2, Maximum=1.2)
    )
    Index 2 Param 3=&FloatFromCurve(
     FloatCurve=DataInterface() {propertyValues {Curve {keys [
         {interpMode=RCIM_Cubic, tangentMode=RCTM_Break, time=0, value=1}
         {interpMode=RCIM_Cubic, leaveTangent=11.1233, tangentMode=RCTM_Break, time=1, value=2}
        ]}}}
     Scale Curve=0
    )
    Index 2 Param 4=&FloatFromCurve(
     FloatCurve=DataInterface() {propertyValues {Curve {keys [<at 0 1 RCIM_Cubic>, {interpMode=RCIM_Cubic, leaveTangent=-1.2956, tangentMode=RCTM_Break, time=1, value=0.75}]}}}
     Scale Curve=0
    )
    Module.Param0WriteEnabled=true
    Module.Param0WriteEnabled1=true
    Module.Param0WriteEnabled2=true
    Module.Param1WriteEnabled=false
    Module.Param1WriteEnabled1=true
    Module.Param1WriteEnabled2=true
    Module.Param2WriteEnabled=true
    Module.Param2WriteEnabled1=true
    Module.Param2WriteEnabled2=true
    Module.Param3WriteEnabled=true
    Module.Param3WriteEnabled1=true
    Module.Param3WriteEnabled2=true
    Write Parameter Index 1=true
    Write Parameter Index 2=true
   )
  ]
  rendererClasses [&NiagaraSpriteRendererProperties]
  renderers [{rendererClass=&NiagaraSpriteRendererProperties}]
 }}
summary {
 systemName=NS_Emote_Hoist
 emitters [{emitterName=DustPuff, rendererClasses [&NiagaraSpriteRendererProperties]}]
 userVariables [
  User.bIsFrontEnd {type=NiagaraBool}
  User.bIsFrontEndPreview {type=NiagaraBool}
 ]
}
