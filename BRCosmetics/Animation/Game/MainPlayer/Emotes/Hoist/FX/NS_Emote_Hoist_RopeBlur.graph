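# Niagara effect for the Hoist emote that draws a motion-blurred rope as a mesh particle.
# System scripts sample a skeletal mesh and supply rope width, radius, angle and opacity through scratch and set-variable modules.
# The RopeBlur emitter burst-spawns a persistent mesh that follows only the sampled X position relative to the emitter.
# Those values drive mesh scale, dynamic material parameters and alpha, while time of day scales brightness.
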NiagaraSystem /BRCosmetics/Animation/Game/MainPlayer/Emotes/Hoist/FX/NS_Emote_Hoist_RopeBlur
dependencies {
 usedDataInterfaces [&NiagaraDataInterfaceSkeletalMesh]
 usedDynamicInputs [
  &Add_Integer
  &ModuloInt
  &ClampFloat
  &Multiply_Float
  &Divide_Float
  &VectorLength
  &DistanceBetweenPositions
 ]
 usedModules [
  &self:SystemSpawnScript.SystemScriptSource.SystemScriptGraph.NiagaraNodeAssignment_3.SetVariables_#1
  &SystemState
  &SampleSkeletalMesh
  &self:ScratchModule_02
  &self:SystemSpawnScript.SystemScriptSource.SystemScriptGraph.NiagaraNodeAssignment_0.SetVariables_#2
  &self:SystemSpawnScript.SystemScriptSource.SystemScriptGraph.NiagaraNodeAssignment_1.SetVariables_#3
 ]
}
emitters {RopeBlur {
  emitterUpdateScript [
   &EmitterState(Life Cycle Mode=::NewEnumerator1, Loop Behavior=::NewEnumerator1, Loop Duration Mode=::NewEnumerator1, UseLoopDelay=false)
   &SpawnBurst_Instantaneous(Use Spawn Probability=false, UseLoopCountLimit=false)
  ]
  particleSpawnScript [&InitializeParticle(Color=<rgba 0.65 0.4 1 0.5>, Lifetime=5, Mesh Scale Mode=::NewEnumerator3, UsePositionOffset=false, Write Material Random=true)]
  particleUpdateScript [
   &ParticleState(Kill Particles When Lifetime Has Elapsed=false)
   &ScaleMeshSize(Scale Factor=&MakeVector(X=$System.RopeWidth, Y=$System.RopeRadius, Z=$System.RopeRadius))
   &DynamicMaterialParameters(
    Index 0 Param 1=&Subtract_Float(A=$System.RopeAngle, B=0.04)
    Index 0 Param 2=0.1
    Module.Index0Param0Write=true
    Module.Index0Param1Write=true
    Module.Index0Param2Write=true
    Module.Index0Param3Write=true
   )
   &ScaleColor(
    Scale Alpha=$System.Opacity
    Scale RGB=&VectorFromFloat(Value=&FNTimeOfDayScalar(Night Scale=0.5, Morning Scale=0.8, Evening Scale=0.7))
    ScaleA=true
    ScaleRGB=true
   )
   /SetVariables(Particles.Position=&AddVectorToPosition(
     Position=$Engine.Emitter.SimulationPosition
     Vector=&MakeVector(
      X=&MakeFloatFromVector(Vector=$System.SampleSkeletalMesh.SampledPosition, Channel=::NewEnumerator0)
      Y=0
      Z=0
     )
    )) {isSetParameters=true}
  ]
  rendererClasses [&NiagaraMeshRendererProperties]
  renderers [{rendererClass=&NiagaraMeshRendererProperties, rendererIndex=0}]
 }}
summary {
 systemName=NS_Emote_Hoist_RopeBlur
 emitters [{emitterName=RopeBlur, rendererClasses [&NiagaraMeshRendererProperties]}]
 userVariables [
  User.bIsFrontEnd {type=NiagaraBool}
  User.bIsFrontEndPreview {type=NiagaraBool}
 ]
}
